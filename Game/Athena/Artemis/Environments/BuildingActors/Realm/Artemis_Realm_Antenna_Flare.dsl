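; Realm antenna building actor Blueprint that switches its emissive glow and satellite flare on or off.
; BeginPlay creates a dynamic material instance for the antenna mesh, then turns the light on if ForceActivated, otherwise re-sets Activated to its current value.
; OnRep_Activated applies that replicated state, either calling TurnOnLight to raise EmissiveBrightness and show the flare, or zeroing the brightness and hiding it.
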(event EventBeginPlay
  (bind _returnvalue (Rendering|Material|CreateDynamicMaterialInstance (Variables|Debug|GetStaticMeshComponent) 0 "/Game/Environments/Artemis/Props/Realm/Realm_Antenna_A/Materials/MI_Realm_Antenna_A.MI_Realm_Antenna_A" "AntEm" :!class /Script/Engine.PrimitiveComponent))
  (Variables|Default|SetDMI_Antenna _returnvalue)
  (if (Variables|Default|GetForceActivated)
    (CallFunction|TurnonLight)
    (else
      (Variables|Default|SetActivated (Variables|Default|GetActivated)))))

(event Custom|TurnOnLight
  (Rendering|Material|SetScalarParameterValue (Variables|Default|GetDMI_Antenna) "EmissiveBrightness" 1e+01 :!class /Script/Engine.MaterialInstanceDynamic)
  (Development|SetHiddeninGame (Variables|Default|GetSatellite_Flare) false false)
  (Rendering|SetVisibility (Variables|Default|GetSatellite_Flare) true false))

(fn OnRep_Activated
  (if (Variables|Default|GetActivated)
    (CallFunction|TurnonLight)
    (else
      (Rendering|Material|SetScalarParameterValue (Variables|Default|GetDMI_Antenna) "EmissiveBrightness" 0.0 :!class /Script/Engine.MaterialInstanceDynamic)
      (Development|SetHiddeninGame (Variables|Default|GetSatellite_Flare) true false)
      (Rendering|SetVisibility (Variables|Default|GetSatellite_Flare) false false))))

(fn ConstructionScript)
